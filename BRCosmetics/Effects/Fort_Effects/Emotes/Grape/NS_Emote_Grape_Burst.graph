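# One-shot Niagara burst for the Grape emote that throws out vines, wood chips, smoke, leaves and tumbling meshes.
# VinesBurst grows ribbons along per-ribbon seeded random directions, while chips and smoke fly out under drag.
# Rolly_Ball uses three spawn groups to tag mesh visibility, with fading scale and spin.
# Leaf_GPU splits by spawn group into short-lived burst leaves and longer-lived drifters swayed by sine waves.
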NiagaraSystem /BRCosmetics/Effects/Fort_Effects/Emotes/Grape/NS_Emote_Grape_Burst
dependencies {
 usedDataInterfaces [&NiagaraDataInterfaceCurve]
 usedDynamicInputs []
 usedModules [
  &self:SystemSpawnScript.SystemScriptSource.SystemScriptGraph.NiagaraNodeAssignment_0.SetVariables_#1
  &SystemState
 ]
}
emitters {
 Leaf_GPU {
  emitterUpdateScript [
   &EmitterState()
   &SpawnBurst_Instantaneous(Spawn Count=10, Use Spawn Probability=false)
   &SpawnBurst_Instantaneous(Spawn Count=7, Spawn Group=1, Use Spawn Probability=false) {module=SpawnBurst_Instantaneous001}
  ]
  particleSpawnScript [
   /Niagara/Modules/Spawn/Initialization/InitializeParticle(
    Color=&UniformRangedLinearColor(Minimum=<rgba 0.723 0.6164 0.3355 3>, Maximum=<rgba 0.9453 1 0.453 3>)
    Lifetime=/Game/Effects/Niagara/DynamicInputs/FN_PickFloatFromSpawnGroupCutoff(
     Over Cutoff Float:NiagaraFloat=&UniformRangedFloat(Minimum=0.5, Maximum=3)
     Under Cutoff Float:NiagaraFloat=&UniformRangedFloat(Minimum=0.2, Maximum=0.75)
     SpawnGroup:NiagaraInt32=$Emitter.SpawnGroup
     Cutoff:NiagaraInt32=1
    )
    Module.Write Color=true
    Module.Write Lifetime=true
    Module.Write Mass=false
    Module.Write Material Random=true
    Module.Write Position=false
    Module.Write Scale=false
    Module.Write SpriteRotation=true
    Module.Write SpriteSize=true
    Position=&SimulationPosition()
    Sprite Rotation=&UniformRangedFloat(Minimum=0, Maximum=360)
    Sprite Size=&Vector2DFromFloat(Value=&UniformRangedFloat(Minimum=6, Maximum=15))
   )
   /SetVariables(
    Particles.SpriteFacing=&UniformRangedVector(Minimum=<-1 -1 -1>, Maximum=<1 1 1>)
    Particles.SpriteInitRotationRate:NiagaraFloat=&Multiply_Float(A=&UniformAOrBFloat(A=-1), B=&UniformRangedFloat(Minimum=3, Maximum=7))
   ) {isSetParameters=true}
   &AddVelocity(Velocity=&Multiply_VectorByFloat(
     Vector=/Game/Effects/Niagara/DynamicInputs/Water/FN_PickVectorFromSpawnGroupCutoff(
      Under Cutoff Vector:Vector3f=&UniformRangedVector(Minimum=<-1 -1 0>, Maximum=<1 0 1>)
      Over Cutoff Vector:Vector3f=&UniformRangedVector(Minimum=<-1 -1 0>, Maximum=<1 -0.1 1>)
      SpawnGroup:NiagaraInt32=$Emitter.SpawnGroup
      Cutoff:NiagaraInt32=1
     )
     Float=100
    ))
   &ApplyInitialForces()
  ]
  particleUpdateScript [
   &ParticleState()
   &ScaleSpriteSize(Scale Factor=&MakeVector2D(
     X=&FloatFromCurve(FloatCurve=DataInterface() {propertyValues {Curve {keys [{arriveTangent=0.002, interpMode=RCIM_Cubic, time=0.5, value=1}, <at 1.0011 0.002>]}}})
     Y=1
    ))
   /SetVariables(
    Particles.MultiDirectWave:Vector3f=&MakeVector(
     X=&Sine(Normalized Angle=&Add_Float(A=$Emitter.Age, B=$Particles.Lifetime), Period=2, Scale=1, Bias=0)
     Y=&Sine(Normalized Angle=&Add_Float(A=$Emitter.Age, B=$Particles.Lifetime), Period=2, Scale=1, Bias=0)
     Z=&Multiply_Float(
      A=0.1
      B=&Sine(
       Normalized Angle=&Add_Float(A=0.125, B=&Add_Float(A=$Emitter.Age, B=$Particles.Lifetime))
       Period=2
       Scale=1
       Bias=-1
      )
     )
    )
    Particles.RandomizeWaveByFacing:Vector3f=&Multiply_VectorByFloat(
     Vector=&Add_Vector(A=<0 0 1>, B=&Multiply_Vector(A=<1 1 0>, B=$Particles.Initial.SpriteFacing))
     Float=&FloatFromCurve(FloatCurve=DataInterface() {propertyValues {Curve {keys [{arriveTangent=2.2462, interpMode=RCIM_Cubic, leaveTangent=2.2462, tangentMode=RCTM_User, time=0, value=0}, <at 1 1>]}}})
    )
    Particles.SineInfluence:Vector3f=&Multiply_VectorByFloat(Vector=<1 1 0.5>, Float=100)
   ) {isSetParameters=true}
   /SetVariables(
    Particles.BurstVelocity:Vector3f=&Multiply_Vector(
     A=$Particles.Initial.Velocity
     B=&Lerp_Vector(
      A=<5 5 5>
      B=<0.2 0.2 -1>
      Alpha=&FloatFromCurve(FloatCurve=DataInterface() {propertyValues {Curve {keys [{arriveTangent=2.8925, interpMode=RCIM_Cubic, leaveTangent=2.8925, tangentMode=RCTM_User, time=0, value=0}, <at 1 1 RCIM_Cubic>]}}})
     )
    )
    Particles.DriftVelocityBase:Vector3f=&Multiply_Vector(
     A=$Particles.Initial.Velocity
     B=&Lerp_Vector(
      A=<2 2 0.5>
      B=<0 0 -2>
      Alpha=&FloatFromCurve(FloatCurve=DataInterface() {propertyValues {Curve {keys [{arriveTangent=3.7448, interpMode=RCIM_Cubic, leaveTangent=3.7448, tangentMode=RCTM_User, time=0, value=0}, <at 1 1>]}}})
     )
    )
    Particles.DriftVelocityWave:Vector3f=&Multiply_Vector(
     A=&Multiply_Vector(A=$Particles.SineInfluence, B=$Particles.MultiDirectWave)
     B=$Particles.RandomizeWaveByFacing
    )
   ) {isSetParameters=true}
   /SetVariables(Particles.Velocity=&FN_PickVectorFromSpawnGroupCutoff(
     Under Cutoff Vector:Vector3f=$Particles.BurstVelocity
     Over Cutoff Vector:Vector3f=&Add_Vector(A=$Particles.DriftVelocityBase, B=$Particles.DriftVelocityWave)
     SpawnGroup:NiagaraInt32=$Emitter.SpawnGroup
     Cutoff:NiagaraInt32=1
    )) {isSetParameters=true}
   &SolveForcesAndVelocity(Clamp Velocity=false, Limit Acceleration=false)
   &SpriteRotationRate(Rotation Rate=&Multiply_Float(A=$Particles.SpriteInitRotationRate, B=&VectorLength(Vector=$Particles.Velocity)))
   /SetVariables(Particles.SpriteFacing=&Add_Vector(
     A=$Particles.Initial.SpriteFacing
     B=&Multiply_VectorByFloat(
      Vector=/Niagara/DynamicInputs/Helpers/SafeNormalizeVector(Vector To Normalize:Vector3f=Linked() {linkedVariable {name=Particles.Position, type=NiagaraPosition}}, Fallback Vector:Vector3f=<0 0 0>)
      Float=0.5
     )
    )) {isSetParameters=true}
   &SolveRotationalForcesAndVelocity()
  ]
  rendererClasses [&NiagaraSpriteRendererProperties]
  renderers [{rendererClass=&NiagaraSpriteRendererProperties}]
 }
 Rolly_Ball {
  emitterSpawnScript [/SetVariables(Emitter.EmitterRandom:NiagaraInt32=&UniformRangedInt(Minimum=0, Maximum=10000)) {isSetParameters=true}]
  emitterUpdateScript [
   &EmitterState()
   &SpawnBurst_Instantaneous(Spawn Count=3, Use Spawn Probability=false)
   &SpawnBurst_Instantaneous(Spawn Count=3, Spawn Group=1, Use Spawn Probability=false) {module=SpawnBurst_Instantaneous001}
   &SpawnBurst_Instantaneous(Spawn Count=3, Spawn Group=2, Use Spawn Probability=false) {module=SpawnBurst_Instantaneous002}
  ]
  particleSpawnScript [
   &InitializeParticle(
    Lifetime=&UniformRangedFloat(Minimum=0.1, Maximum=0.75)
    Mesh Scale=&VectorFromFloat(Value=1)
    Module.Write Color=true
    Module.Write Lifetime=true
    Module.Write Mass=true
    Module.Write Material Random=true
    Module.Write Position=true
    Module.Write Scale=true
    Module.Write SpriteRotation=false
    Module.Write SpriteSize=true
    Position=&Add_Vector(A=/Niagara/DynamicInputs/Helpers/SimulationPosition(), B=<0 0 0>)
    Sprite Size=<5 5>
   )
   &SystemLocation(MaskSpawnGroup=false, Offset=<0 -10 -10>)
   /SetVariables(Particles.Scale=&VectorFromFloat(Value=/Niagara/DynamicInputs/Math/LerpMultipleFloats(
      Float Count:ENiagara_MultipleLerpCount=::NewEnumerator0
      1:NiagaraFloat=0.1
      2:NiagaraFloat=0.3
      3:NiagaraFloat=0.1
      4:NiagaraFloat=?
      5:NiagaraFloat=?
      Alpha:NiagaraFloat=$Particles.Lifetime
     ))) {isSetParameters=true}
   /SetVariables(Particles.VisibilityTag=&SpawnGroup()) {isSetParameters=true}
   /Niagara/Modules/Spawn/Orientation/V2/InitialMeshOrientation(
    Orientation Vector=&UniformRangedVector(Minimum=<-1 -1 -1>, Maximum=<1 1 1>)
    Use Orientation Vector=true
    Use Rotation Vector=false
   )
  ]
  particleUpdateScript [
   &ParticleState()
   &ScaleColor(
    Scale Alpha=&FloatFromCurve(FloatCurve=DataInterface() {propertyValues {Curve {keys [
         {interpMode=RCIM_Cubic, leaveTangent=-0.0663, tangentMode=RCTM_Break, time=0, value=1}
         {arriveTangent=-0.9951, interpMode=RCIM_Cubic, leaveTangent=-0.9951, tangentMode=RCTM_User, time=1, value=0}
        ]}}})
    ScaleA=true
    ScaleRGB=true
   )
   /Niagara/Modules/Update/Orientation/V2/UpdateMeshOrientation(
    Rotation Rate:NiagaraFloat=&FloatFromCurve(
     FloatCurve=DataInterface() {propertyValues {Curve {keys [
         {arriveTangent=-1.9834, interpMode=RCIM_Cubic, leaveTangent=-1.9834, tangentMode=RCTM_User, time=0, value=1}
         {arriveTangent=-0.2717, interpMode=RCIM_Cubic, leaveTangent=-0.2717, tangentMode=RCTM_User, time=1, value=0}
        ]}}}
     Scale Curve=&UniformAOrBFloat(
      A=-1
      Randomness Mode=::NewEnumerator1
      Random Seed=&Add_Integer(A=$Emitter.EmitterRandom, B=&SpawnGroup(Max Spawn Group=2))
      Fixed Random Seed=true
     )
    )
    Rotation Vector:Vector3f=&UniformRangedVector(Minimum=<0.5 0.5 0.5>, Maximum=<1 1 1>)
   )
   &ScaleMeshSize(Scale Factor=&VectorFromFloat(Value=&FloatFromCurve(FloatCurve=DataInterface() {propertyValues {Curve {keys [
          {arriveTangent=2.9889, interpMode=RCIM_Cubic, leaveTangent=2.9889, tangentMode=RCTM_User, time=0, value=0}
          {arriveTangent=0.1698, interpMode=RCIM_Cubic, leaveTangent=0.1698, tangentMode=RCTM_User, time=0.5, value=1.0115}
          {arriveTangent=-1.9119, interpMode=RCIM_Cubic, leaveTangent=-1.9119, tangentMode=RCTM_User, time=0.9989, value=0.5623}
         ]}}})))
   /SetVariables(Particles.ScaleNormalized:NiagaraFloat=&Add_Float(
     A=-0.5
     B=&Multiply_Float(A=5, B=&MakeFloatFromVector(Vector=$Particles.Scale, Channel=::NewEnumerator0))
    )) {isSetParameters=true}
   /SetVariables(Particles.NewPosition:Vector3f=&Add_Vector(
     A=Linked() {linkedVariable {name=Particles.Position, type=NiagaraPosition}}
     B=&Lerp_Vector(A=<0 20 0>, B=<0 0 0>, Alpha=$Particles.ScaleNormalized)
    )) {isSetParameters=true}
  ]
  rendererClasses [&NiagaraMeshRendererProperties]
  renderers [
   {rendererClass=&NiagaraMeshRendererProperties, rendererIndex=0}
   {rendererClass=&NiagaraMeshRendererProperties, rendererIndex=1}
   {rendererClass=&NiagaraMeshRendererProperties, rendererIndex=2}
  ]
 }
 Smoke {
  emitterUpdateScript [
   &EmitterState()
   &SpawnBurst_Instantaneous(Spawn Count=8, Use Spawn Probability=false)
  ]
  particleSpawnScript [
   &InitializeParticle(
    Color=&UniformRangedLinearColor(Minimum=<rgba 0.092 0.0681 0.0553 0.3>, Maximum=<rgba 0.159 0.1317 0.0771 1>)
    Lifetime=&UniformRangedFloat(Minimum=0.3, Maximum=0.6)
    Module.Write Color=true
    Module.Write Lifetime=true
    Module.Write Mass=true
    Module.Write Material Random=true
    Module.Write Position=false
    Module.Write Scale=false
    Module.Write SpriteRotation=true
    Module.Write SpriteSize=true
    Position=/Niagara/Functions/Localspace/SimulationPosition()
    Sprite Rotation=&UniformRangedFloat(Minimum=0, Maximum=360)
    Sprite Size=&Vector2DFromFloat(Value=&UniformRangedFloat(Minimum=10, Maximum=40))
   )
   &SystemLocation(MaskSpawnGroup=false, Offset=<0 -15 -20>)
   &AddVelocity(Velocity=&Multiply_VectorByFloat(Vector=&UniformRangedVector(Minimum=<-1 -1 -0.25>, Maximum=<1 -0.2 1>), Float=150))
  ]
  particleUpdateScript [
   &ParticleState()
   &Drag(Drag=3, Use Linear Drag=true, Use Rotational Drag=false)
   &SolveForcesAndVelocity(Clamp Velocity=false, Limit Acceleration=false)
   &ScaleSpriteSize(Scale Factor=&Vector2DFromFloat(Value=&FloatFromCurve(FloatCurve=DataInterface() {propertyValues {Curve {keys [{arriveTangent=9.5066, interpMode=RCIM_Cubic, leaveTangent=9.5066, tangentMode=RCTM_User, time=0, value=0.25}, <at 1.0011 3.4809 RCIM_Cubic>]}}})))
   /SetVariables(Particles.SubImageIndex=&FloatFromCurve(FloatCurve=DataInterface() {propertyValues {Curve {keys [
         {arriveTangent=56.912, interpMode=RCIM_Cubic, leaveTangent=56.912, tangentMode=RCTM_User, time=-0.0046, value=5}
         <at 1.0011 22.1689 RCIM_Cubic>
        ]}}})) {isSetParameters=true}
   &DynamicMaterialParameters(
    Index 0 Param 1=&FloatFromCurve(FloatCurve=DataInterface() {propertyValues {Curve {keys [<at 0.3666 1.0113 RCIM_Cubic>,<at 1.0011 -0.8759 RCIM_Cubic>]}}})
    Index 0 Param 4=&FloatFromCurve(FloatCurve=DataInterface() {propertyValues {Curve {keys [<at -0.0046 0.2 RCIM_Cubic>,<at 1.0011 1 RCIM_Cubic>]}}})
    Module.Param0WriteEnabled=true
    Module.Param1WriteEnabled=false
    Module.Param2WriteEnabled=false
    Module.Param3WriteEnabled=true
   )
   &ScaleColor(
    Scale Alpha=&FloatFromCurve(FloatCurve=DataInterface() {propertyValues {Curve {keys [{arriveTangent=-1.8634, interpMode=RCIM_Cubic, leaveTangent=-1.8634, tangentMode=RCTM_User, time=0, value=1}, <at 1 0.25>]}}})
    ScaleA=true
    ScaleRGB=false
   )
  ]
  rendererClasses [&NiagaraSpriteRendererProperties]
  renderers [{rendererClass=&NiagaraSpriteRendererProperties}]
 }
 VinesBurst {
  emitterSpawnScript [/SetVariables(Emitter.EmitterRandom:NiagaraInt32=&UniformRangedInt(Minimum=0, Maximum=10000)) {isSetParameters=true}]
  emitterUpdateScript [
   &EmitterState(Life Cycle Mode=::NewEnumerator1, Loop Behavior=::NewEnumerator1, Loop Delay=0.1, Loop Duration=1, UseLoopDelay=true)
   &FN_SpawnRibbon_v2(Number of Ribbons=7, SpawnRate:NiagaraFloat=&UniformRangedFloat(Minimum=5, Maximum=10))
  ]
  particleSpawnScript [
   &FN_InitializeRibbon_v3(
    Color:LinearColor=&UniformRangedLinearColor(
     Minimum=&UniformRangedLinearColor(Minimum=<rgba 0.1069 0.12 0.0763 1>, Maximum=<rgba 0.156 0.1018 0.0512 1>)
     Maximum=&UniformRangedLinearColor(Minimum=<rgba 0.253 0.1536 0.0918 1>, Maximum=<rgba 0.195 0.1475 0.0653 1>)
    )
    Lifetime:NiagaraFloat=&UniformRangedFloat(
     Minimum=0.5
     Maximum=1
     Randomness Mode=::NewEnumerator1
     Random Seed=&Add_Integer(A=$Emitter.EmitterRandom, B=&self:NiagaraRibbonToInt())
     Fixed Random Seed=true
    )
    Ribbon Width:NiagaraFloat=&UniformRangedFloat(Minimum=12, Maximum=17)
   )
   /SetVariables(Particles.RibbonDirection:Vector3f=&SafeNormalizeVector(
     Vector To Normalize:Vector3f=&UniformRangedVector(
      Minimum=<-1 -1 -1>
      Maximum=<1 -0.01 1>
      Randomness Mode=::NewEnumerator1
      Random Seed=&Add_Integer(A=$Emitter.EmitterRandom, B=&self:NiagaraRibbonToInt())
      Fixed Random Seed=true
     )
     Fallback Vector:Vector3f=<0 0 0>
    )) {isSetParameters=true}
   /SetVariables(Particles.RibbonLength:NiagaraFloat=&UniformRangedFloat(
     Minimum=0.2
     Maximum=1
     Randomness Mode=::NewEnumerator1
     Random Seed=&Add_Integer(A=$Emitter.EmitterRandom, B=&self:NiagaraRibbonToInt())
     Fixed Random Seed=true
    )) {isSetParameters=true}
   /SetVariables(Particles.CurrentPlacement:Vector3f=&Multiply_Vector(
     A=$Particles.RibbonDirection
     B=&VectorFromFloat(Value=&Multiply_Float(
       A=100
       B=&Multiply_Float(
        A=&FloatFromCurve(
         FloatCurve=DataInterface() {propertyValues {Curve {keys [
             {arriveTangent=-3.8297, interpMode=RCIM_Cubic, leaveTangent=-3.8297, tangentMode=RCTM_User, time=0, value=1}
             {arriveTangent=-0.4062, interpMode=RCIM_Cubic, leaveTangent=-0.4062, tangentMode=RCTM_User, time=0.2825, value=0.2598}
             {arriveTangent=-0.3287, interpMode=RCIM_Cubic, leaveTangent=-0.3287, tangentMode=RCTM_User, time=1, value=0}
            ]}}}
         CurveIndex=$Emitter.NormalizedLoopAge
        )
        B=$Particles.RibbonLength
       )
      ))
    )) {isSetParameters=true}
   /SetVariables(Particles.SpriteSize=<2 2>) {isSetParameters=true}
   /Niagara/Modules/Spawn/Location/SphereLocation(
    MaskSpawnGroup=false
    Module.Override Randomness=false
    Offset=$Particles.CurrentPlacement
    Sphere Radius=&Multiply_Float(
     A=5
     B=&Multiply_Float(
      A=&FloatFromCurve(FloatCurve=DataInterface() {propertyValues {Curve {keys [<at 0 1>,<at 1 0>]}}}, CurveIndex=$Emitter.NormalizedLoopAge)
      B=$Particles.RibbonLength
     )
    )
    Surface Only=false
   ) {module=SphereLocation003}
   &SphereLocation(
    MaskSpawnGroup=false
    Module.Override Randomness=false
    Offset=&Multiply_Vector(
     A=$Particles.RibbonDirection
     B=&VectorFromFloat(Value=&Multiply_Float(
       A=100
       B=&Multiply_Float(A=$Particles.RibbonLength, B=&Subtract_Float(A=1, B=$Emitter.NormalizedLoopAge))
      ))
    )
    Sphere Radius=0
    Surface Only=false
   ) {enabled=false}
   &AddVelocityFromPoint(
    Module.Use Falloff Distance=false
    Origin Offset:NiagaraPosition=$Particles.Position
    Velocity Strength=&Multiply_Float(A=200, B=$Particles.RibbonLength)
   ) {enabled=false}
   &AddVelocity(Velocity=&Multiply_VectorByFloat(
     Vector=&UniformRangedVector(Minimum=<-1 -1 -1>, Maximum=<1 0.5 1>)
     Float=&FloatFromCurve(
      FloatCurve=DataInterface() {propertyValues {Curve {keys [<at 0 1 RCIM_Cubic>,<at 1 0>]}}}
      CurveIndex=$Emitter.NormalizedLoopAge
      Scale Curve=500
     )
    ))
  ]
  particleUpdateScript [
   &ParticleState()
   &ScaleVelocity(Velocity Scale=&VectorFromFloat(Value=&FloatFromCurve(
      FloatCurve=DataInterface() {propertyValues {Curve {keys [
          {arriveTangent=-6.2948, interpMode=RCIM_Cubic, leaveTangent=-6.2948, tangentMode=RCTM_User, time=0, value=1}
          {arriveTangent=-0.9463, interpMode=RCIM_Cubic, leaveTangent=-0.9463, tangentMode=RCTM_User, time=0.1217, value=0.5428}
          <at 1 0.25>
         ]}}}
      Scale Curve=$Particles.RibbonLength
     )))
   &SolveForcesAndVelocity(Clamp Velocity=false, Limit Acceleration=false)
   /Game/Effects/Niagara/Modules/Contrails/FN_UpdateRibbon/FN_UpdateRibbon_v3()
   /Game/Effects/Niagara/Modules/Materials/MaterialPanner(
    Frequency:NiagaraFloat=&Lerp_Float(A=0.19, B=5, Alpha=$Particles.RibbonLength)
    Jitter Multiplier=0
    Jitter Pan by SpawnGroups=false
    Speed:NiagaraFloat=&FloatFromCurve(FloatCurve=DataInterface() {propertyValues {Curve {keys [<at 0 0>,<at 1 -2>]}}})
   ) {module=MaterialPanner001}
   &DynamicMaterialParameters(
    Index 0 Param 1=$Particles.Age
    Index 0 Param 2=&UniformRangedFloat(Minimum=0, Maximum=1)
    Module.Param0WriteEnabled=true
    Module.Param1WriteEnabled=true
    Module.Param2WriteEnabled=true
    Module.Param3WriteEnabled=true
   )
   &ScaleRibbonWidth(Ribbon Width Scale=&Multiply_Float(
     A=$Particles.RibbonLength
     B=&FloatFromCurve(FloatCurve=DataInterface() {propertyValues {Curve {keys [
          {interpMode=RCIM_Cubic, tangentMode=RCTM_User, time=0, value=1}
          {arriveTangent=-2.0438, interpMode=RCIM_Cubic, leaveTangent=-2.0438, tangentMode=RCTM_User, time=1, value=0}
         ]}}})
    ))
   &ScaleColor(
    Scale RGB=&Lerp_Vector(
     A=<2 2 2>
     B=&MakeVectorFromLinearColorRGB(Color=<rgba 0.558 0.3571 0.3169 1>)
     Alpha=$Particles.NormalizedAge
    )
    ScaleA=true
    ScaleRGB=true
   )
  ]
  rendererClasses [&NiagaraRibbonRendererProperties, &NiagaraSpriteRendererProperties]
  renderers [
   {rendererClass=&NiagaraRibbonRendererProperties}
   {rendererClass=&NiagaraSpriteRendererProperties, rendererIndex=1}
  ]
 }
 WoodChipsMesh_v2 {
  emitterUpdateScript [
   &EmitterState(
    Enable Distance Culling=true
    Max Distance Response=::KillImmediately
    MaxDistance=2000
    Module.bUseMaxDistance=true
    Module.bUseMinDistance=false
    Scalability Mode=::NewEnumerator1
   )
   &SpawnBurst_Instantaneous(Spawn Count=&UniformRangedInt(Minimum=10, Maximum=20), Use Spawn Probability=false)
  ]
  particleSpawnScript [
   &InitializeParticle(
    Lifetime=&UniformRangedFloat(Minimum=0.25, Maximum=0.5)
    Mesh Scale=&Multiply_VectorByFloat(
     Vector=&UniformRangedVector(Minimum=<0.1 0.1 0.1>, Maximum=<0.5 0.5 0.5>)
     Float=&UniformRangedFloat(Minimum=0.75, Maximum=1.5)
    )
    Module.Write Color=false
    Module.Write Lifetime=true
    Module.Write Mass=false
    Module.Write Material Random=true
    Module.Write Position=false
    Module.Write Scale=true
    Module.Write SpriteRotation=false
    Module.Write SpriteSize=false
    Position=&SimulationPosition()
   )
   &InitialMeshOrientation(
    Orientation Vector=/Niagara/DynamicInputs/Random/RandomVector(
     Vector Scale:NiagaraFloat=1
     Constraint Vector:Vector3f=<0 0 1>
     Randomness Mode:ENiagaraRandomnessMode=::NewEnumerator0
     Random Seed:NiagaraInt32=0
     Fixed Random Seed:NiagaraBool=false
    )
    Use Orientation Vector=true
    Use Rotation Vector=false
   )
   &AddRotationalVelocity(Rotation Rate=&RandomVector(
     Vector Scale:NiagaraFloat=1
     Constraint Vector:Vector3f=<0 0 1>
     Randomness Mode:ENiagaraRandomnessMode=::NewEnumerator0
     Random Seed:NiagaraInt32=0
     Fixed Random Seed:NiagaraBool=false
    ))
   &AddVelocity(Velocity=&Multiply_VectorByFloat(Vector=&UniformRangedVector(Minimum=<-1 -1 -0.25>, Maximum=<1 -0.1 1>), Float=400))
  ]
  particleUpdateScript [
   &ParticleState()
   &Drag(Drag=3, Use Linear Drag=true, Use Rotational Drag=false)
   &GravityForce(Gravity=&UniformRangedVector(Minimum=<0 0 -100>, Maximum=<1 1 -980>))
   &SolveForcesAndVelocity(Clamp Velocity=false, Limit Acceleration=false)
   &SolveRotationalForcesAndVelocity()
  ]
  rendererClasses [&NiagaraMeshRendererProperties]
  renderers [{rendererClass=&NiagaraMeshRendererProperties, rendererIndex=0}]
 }
}
summary {
 systemName=NS_Emote_Grape_Burst
 emitters [
  {emitterName=VinesBurst, rendererClasses [&NiagaraRibbonRendererProperties, &NiagaraSpriteRendererProperties]}
  {emitterName=WoodChipsMesh_v2, rendererClasses [&NiagaraMeshRendererProperties]}
  {emitterName=Smoke, rendererClasses [&NiagaraSpriteRendererProperties]}
  {emitterName=Rolly_Ball, rendererClasses [&NiagaraMeshRendererProperties]}
  {emitterName=Leaf_GPU, rendererClasses [&NiagaraSpriteRendererProperties]}
 ]
 userVariables [
  User.bIsFrontEndPreview {type=NiagaraBool}
  User.GlobalOffset {type=Vector3f, description="", defaultValue {struct=Vector3f, value=<0 0 0>}}
  User.PlayerSpeed {type=NiagaraFloat, defaultValue=0, description=""}
  User.RibbonIDOffset {type=NiagaraInt32}
  User.SkeletalMesh {
   type=NiagaraDataInterfaceSkeletalMesh
   defaultValue {
    struct=NiagaraExt_VariableValue_DataInterface
    value {dataInterface=&self:NiagaraDataInterfaceSkeletalMesh_0}
   }
  }
 ]
}
